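# Looping ambient lake water sound cue in the BlastBerry map content.
# A single looping wave player feeds the output directly with the water_lake_01b wave from the base game's ambient sounds.
# It plays in the Fort_Ambient_Water sound class, with its own concurrency override that prevents new instances once the limit is reached.
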SoundCue /BlastBerryMapContent/Game/Sounds/Fort_Ambient_Looping/Water/water_lake_01b_Cue
root=WavePlayer_476
nodes [{id=WavePlayer_476, properties {bLooping=true, soundWaveAssetPtr=/Game/Sounds/Fort_Ambient_Looping/Water/water_lake_01b}}]
properties {
 bOverrideConcurrency=true
 maxDistance=2097152
 soundClassObject=/Game/Sounds/FortSoundClassesAndModes/Fort_Ambient_Water
 concurrencyOverrides {resolutionRule=PreventNew}
}
